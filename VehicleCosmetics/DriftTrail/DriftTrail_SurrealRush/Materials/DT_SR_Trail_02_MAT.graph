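# Unlit translucent Niagara material for the Surreal Rush drift trail, usable on ribbons, sprites and meshes.
# Panning flow and cell noise form the body, eroded by circle, tip and blob masks, then shaped by U falloffs, depth and camera fades.
# Flow intensity thresholds particle color into stepped bands, with static switches choosing texture and color modes.
# Dynamic parameters drive flow offset and speed, tail mask speed, and the Color2 paint blend.
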Material /VehicleCosmetics/DriftTrail/DriftTrail_SurrealRush/Materials/DT_SR_Trail_02_MAT
graph {
 properties {
  bAutomaticallySetUsageInEditor=false
  bUsedWithNiagaraMeshParticles=true
  bUsedWithNiagaraRibbons=true
  bUsedWithNiagaraSprites=true
  blendMode=BLEND_Translucent
  pixelDepthOffsetMode=PDOM_Legacy
  shadingModel=MSM_Unlit
  usedShadingModels=Unlit
 }
 nodes [
  n1 = TextureCoordinate()
  n2 = Multiply(
   A=TextureSample(UVs=Panner(Coordinate=@n1) {speedX=0.3, speedY=0.2}) {texture=/DealershipSharedLibrary/FX/Textures/Noise/T_VCVFX_CellNoise2_A}.RGB
   B=Constant() {r=8}
  )
  n3 = Constant() {r=0.2}
  n4 = TextureCoordinate() {uTiling=3}
  n5 = DynamicParameter() {defaultValue=<rgba 1 1 1 4>, paramNames [offsetTrails, TailMaskSpeed, Painteds, speed]}
  n6 = AppendVector(A=@n5.offsetTrails, B=Constant())
  n7 = Time()
  n8 = TextureSample(UVs=Panner(Coordinate=Add(A=@n4, B=@n6), Time=Multiply(A=@n5.speed, B=@n7)) {speedX=-0.6}) {texture=/DealershipSharedLibrary/FX/Textures/Diffuse/T_VCVFX_AbstracFlow_D}
  n9 = Multiply(A=@n8.R, B=Constant() {r=13})
  n10 = StaticSwitchParameter(True=Multiply(A=Power(Base=@n2, Exp=@n3), B=@n9), False=@n8.B) {defaultValue=true, group=Texture, parameterName=TextureSwitch}
  MaterialExpressionVertexColor_0 = ParticleColor()
  n11 = TextureCoordinate()
  n12 = Subtract(
   A=Multiply(A=@MaterialExpressionVertexColor_0.A, B=@n10)
   B=OneMinus(None=TextureSample(UVs=@n11) {texture=/DealershipSharedLibrary/FX/Textures/Gradient_Masks/T_VCVFX_RibbonBaseCircle_A}.RGB)
  )
  n13 = TextureCoordinate() {vTiling=0.0125}
  n14 = Multiply(A=Time(), B=@n5.TailMaskSpeed)
  n15 = Panner(Coordinate=TextureCoordinate() {uTiling=1.4, vTiling=0.7}) {speedX=-2, speedY=0.25}
  n16 = TextureSample(UVs=Panner(Coordinate=TextureCoordinate()) {speedX=-1, speedY=0.5}) {texture=/DealershipSharedLibrary/FX/Textures/Gradient_Masks/T_VCVFX_BlobNoise_A}
  n17 = TextureCoordinate()
  n18 = Multiply(A=Power(Base=ComponentMask(None=@n17) {r=true}, Exp=Constant() {r=12}), B=Constant() {r=200})
  n19 = Add(
   A=TextureSample(UVs=Panner(Coordinate=@n13, Time=@n14) {speedY=1}) {texture=/DealershipSharedLibrary/FX/Textures/Gradient_Masks/T_VCVFX_RibbonTipMask_A}.RGB
   B=Multiply(
    A=TextureSample(UVs=@n15) {texture=&T_VCVFX_BlobNoise_A}.RGB
    B=Multiply(A=@n16.RGB, B=@n18)
   )
  )
  n20 = Power(Base=ComponentMask(None=TextureCoordinate()) {r=true}, Exp=Constant() {r=1.3})
  n21 = Constant() {r=20}
  n22 = OneMinus(None=ComponentMask(None=TextureCoordinate()) {r=true})
  n23 = Constant() {r=3}
  n24 = Multiply(A=Multiply(A=Subtract(A=@n12, B=@n19), B=Multiply(A=@n20, B=@n21)), B=Multiply(A=Power(Base=@n22, Exp=@n23), B=Constant() {r=20}))
  n25 = MaterialFunctionCall(Opacity In=Constant() {r=1}) {materialFunction=&MF_VCVFX_DepthFade}
  n26 = Constant() {r=100}
  n27 = Constant() {r=50}
  MaterialExpressionConstantClamp_0 = Saturate(None=Multiply(
    A=Multiply(A=@n24, B=@n25.Result)
    B=MaterialFunctionCall(Fade Length=@n26, Fade Offset=@n27) {materialFunction=&CameraDepthFade}.Result
   ))
  Constant() {r=4}
  n28 = Constant() {r=0.02}
  n29 = Multiply(
   A=Power(Base=@n9, Exp=Constant() {r=1.3})
   B=StaticSwitchParameter(True=Constant() {r=2}, False=Constant() {r=3}) {defaultValue=true, group=Texture, parameterName=TriColorSwitch}
  )
  MaterialExpressionConstantClamp_1 = Saturate(None=@n29)
  n30 = Constant() {r=0.55}
  n31 = Constant() {r=2}
  n32 = LinearInterpolate(
   A=LinearInterpolate(A=Multiply(A=@MaterialExpressionVertexColor_0.RGB, B=@n28), B=@MaterialExpressionVertexColor_0.RGB, Alpha=@MaterialExpressionConstantClamp_1)
   B=LinearInterpolate(
    A=Multiply(A=@MaterialExpressionVertexColor_0.RGB, B=@n31)
    B=VectorParameter() {group=Texture, parameterName=Color2, defaultValue=<rgba 1 0.3079 0 0>}.RGB
    Alpha=@n5.Painteds
   )
   Alpha=Step(Y=Constant(), X=Subtract(A=@n29, B=@n30))
  )
  n33 = Multiply(A=@MaterialExpressionVertexColor_0.RGB, B=@n10)
  MaterialExpressionConstantClamp_2 = Saturate(None=Multiply(A=@n33, B=Constant() {r=10}))
  MaterialExpressionConstantClamp_3 = Saturate(None=Subtract(A=@n29, B=Constant() {r=0.3}))
  n34 = LinearInterpolate(A=@n32, B=@MaterialExpressionConstantClamp_2, Alpha=@MaterialExpressionConstantClamp_3)
  MaterialFunctionCall(Emissive=StaticSwitchParameter(True=StaticSwitchParameter(True=@n32, False=@n34) {defaultValue=true, group=Texture, parameterName=TriColorSwitch}, False=@n33) {defaultValue=true, group=Texture, parameterName=SwitchColor}) {materialFunction=&MF_FX_EmissiveAdjustments}
 ]
}
