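# Translucent, unlit, two-sided Niagara material for the wheel-wrap part of the SurrealRush drift trail.
# A panning abstract-flow texture drives layered color bands, with static switches for variants and a dynamic parameter blending in Color2.
# Opacity math erodes the flow with a scrolling tail mask and blob noise, then fades the U ends.
# A hard center-line gradient masks both color and opacity, while dynamic parameters control offset and scroll speeds.
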Material /VehicleCosmetics/DriftTrail/DriftTrail_SurrealRush/Materials/DT_WheelWrap_MAT_1
graph {
 properties {
  bAutomaticallySetUsageInEditor=false
  bUsedWithNiagaraMeshParticles=true
  bUsedWithNiagaraRibbons=true
  bUsedWithNiagaraSprites=true
  blendMode=BLEND_Translucent
  pixelDepthOffsetMode=PDOM_Legacy
  shadingModel=MSM_Unlit
  twoSided=true
  usedShadingModels=Unlit
 }
 nodes [
  Constant() {r=4}
  n1 = Constant() {r=0.02}
  n2 = Multiply(A=Constant3Vector() {constant=<rgba 0.03 0.17 1 0>})
  n3 = DynamicParameter() {defaultValue=<rgba 0 -2 1 4>, paramNames [offsetTrails, TailMaskSpeed, Painteds, speed]}
  n4 = Add(A=TextureCoordinate() {uTiling=3}, B=AppendVector(A=@n3.offsetTrails, B=Constant()))
  n5 = Multiply(A=@n3.speed, B=Time())
  n6 = TextureSample(UVs=MaterialFunctionCall(UVs=Panner(Coordinate=@n4, Time=@n5) {speedX=-2}, Texture Scale=Constant2Vector() {g=0.6, r=1}) {materialFunction=&ScaleUVsByCenter}.UVs) {texture=/DealershipSharedLibrary/FX/Textures/Diffuse/T_VCVFX_AbstracFlow_D}
  n7 = Multiply(A=@n6.R, B=Constant() {r=13})
  n8 = Multiply(
   A=Power(Base=@n7, Exp=Constant() {r=1.3})
   B=StaticSwitchParameter(True=Constant() {r=2}, False=Constant() {r=3}) {defaultValue=true, group=Texture, parameterName=TriColorSwitch}
  )
  n9 = Constant() {r=0.55}
  n10 = Constant() {r=2}
  n11 = LinearInterpolate(
   A=LinearInterpolate(A=Multiply(A=@n2, B=@n1), B=@n2, Alpha=Saturate(None=@n8))
   B=LinearInterpolate(
    A=Multiply(A=@n2, B=@n10)
    B=VectorParameter() {group=Texture, parameterName=Color2, defaultValue=<rgba 0.5027 0 0.0023 0>}.RGB
    Alpha=@n3.Painteds
   )
   Alpha=Step(Y=Constant(), X=Subtract(A=@n8, B=@n9))
  )
  n12 = Multiply(B=Constant() {r=8}) {constA=1}
  n13 = Constant() {r=0.2}
  n14 = StaticSwitchParameter(True=Multiply(A=Power(Base=@n12, Exp=@n13), B=@n7), False=@n6.B) {defaultValue=true, group=Texture, parameterName=TextureSwitch}
  n15 = Multiply(A=@n2, B=@n14)
  n16 = Saturate(None=Multiply(A=@n15, B=Constant() {r=10}))
  n17 = Saturate(None=Subtract(A=@n8, B=Constant() {r=0.3}))
  n18 = StaticSwitchParameter(
   True=StaticSwitchParameter(True=@n11, False=LinearInterpolate(A=@n11, B=@n16, Alpha=@n17)) {defaultValue=true, group=Texture, parameterName=TriColorSwitch}
   False=@n15
  ) {defaultValue=true, group=Texture, parameterName=SwitchColor}
  n19 = TextureSample() {texture=/DealershipSharedLibrary/FX/Textures/Gradient_Masks/T_VCVFX_HardCtrLine_01_A}
  MaterialFunctionCall(Emissive=Multiply(A=Multiply(A=@n18) {constB=2}, B=@n19.RGB)) {materialFunction=&MF_FX_EmissiveAdjustments}
  ParticleColor()
  n20 = TextureCoordinate()
  n21 = ComponentMask(None=@n20) {g=true}
  n22 = OneMinus(None=ComponentMask(None=@n20) {g=true})
  Power(Base=Multiply(A=Power(Base=@n21) {constExponent=3}, B=Power(Base=@n22) {constExponent=3})) {constExponent=0.6}
  n23 = TextureCoordinate()
  n24 = Subtract(
   A=Multiply(A=Constant() {r=0.6}, B=@n14)
   B=OneMinus(None=TextureSample(UVs=@n23) {texture=&T_WhitesRGB}.RGB)
  )
  n25 = Panner(Coordinate=TextureCoordinate() {vTiling=0.0125}, Time=Multiply(A=Time(), B=@n3.TailMaskSpeed)) {speedY=1}
  n26 = Panner(Coordinate=TextureCoordinate() {uTiling=1.4, vTiling=0.7}) {speedX=-2, speedY=0.25}
  n27 = TextureSample(UVs=Panner(Coordinate=TextureCoordinate()) {speedX=-1, speedY=0.5}) {texture=/DealershipSharedLibrary/FX/Textures/Gradient_Masks/T_VCVFX_BlobNoise_A}
  n28 = TextureCoordinate()
  n29 = Multiply(A=Power(Base=ComponentMask(None=@n28) {r=true}, Exp=Constant() {r=12}), B=Constant() {r=200})
  n30 = Add(
   A=OneMinus(None=TextureSample(UVs=@n25) {texture=&T_WhitesRGB}.RGB)
   B=Multiply(
    A=TextureSample(UVs=@n26) {texture=&T_VCVFX_BlobNoise_A}.RGB
    B=Multiply(A=@n27.RGB, B=@n29)
   )
  )
  n31 = Power(Base=ComponentMask(None=TextureCoordinate()) {r=true}, Exp=Constant() {r=0.1})
  n32 = Constant() {r=20}
  n33 = OneMinus(None=ComponentMask(None=TextureCoordinate()) {r=true})
  n34 = Constant() {r=0.1}
  n35 = Multiply(A=Multiply(A=Subtract(A=@n24, B=@n30), B=Multiply(A=@n31, B=@n32)) {constB=5}, B=Multiply(A=Power(Base=@n33, Exp=@n34), B=Constant() {r=20}))
  Reroute(None=Multiply(A=Saturate(None=@n35), B=@n19.RGB))
  TextureSample(UVs=Panner(Coordinate=TextureCoordinate()) {speedX=0.3, speedY=0.2}) {texture=/DealershipSharedLibrary/FX/Textures/Noise/T_VCVFX_CellNoise2_A}
 ]
}
